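# Translucent UI master material that draws a procedural two-color panel for top bar buttons, with normal and selected looks.
# Four edge masks come from offset UVs, rotated by Rotators fed fixed turn values instead of time, then sharpened by softness multipliers.
# Their product, built with scalar dot products, blends each color pair.
# AnimateSelection interpolates edge offsets, tilts, softness and color pairs; a separate Alpha parameter is clamped.
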Material /FortUILibrary/Foundation/Buttons/Materials/TopBar/M_UI_TopBar_PanelButtons_Master
graph {
 properties {
  bUsedWithStaticMesh=false
  blendMode=BLEND_Translucent
  materialDomain=MD_UI
  pixelDepthOffsetMode=PDOM_Legacy
  usedShadingModels=Default Lit
 }
 nodes [
  Clamp(None=ScalarParameter() {defaultValue=1, parameterName=Alpha})
  n1 = ScalarParameter() {defaultValue=-0.0275, parameterName=Angle_Left_Origin}
  n2 = ScalarParameter() {defaultValue=-0.0225, parameterName=Angle_Left_Offset}
  n3 = ScalarParameter() {parameterName=AnimateSelection}
  n4 = Rotator(
   Coordinate=Add(A=TextureCoordinate(), B=LinearInterpolate(A=@n1, B=@n2, Alpha=@n3))
   Time=LinearInterpolate(A=Constant(), B=ScalarParameter() {defaultValue=0.0018, parameterName=Angle_Left}, Alpha=@n3)
  ) {speed=6.2832}
  n5 = ScalarParameter() {defaultValue=128, parameterName=U_Softness_Origin}
  n6 = ScalarParameter() {defaultValue=80, parameterName=U_Softness_Selected}
  n7 = Clamp(None=Multiply(A=ComponentMask(None=@n4) {r=true}, B=LinearInterpolate(A=@n5, B=@n6, Alpha=@n3)))
  n8 = TextureCoordinate()
  n9 = ScalarParameter() {defaultValue=-0.0265, parameterName=Angle_Right_Origin}
  n10 = ScalarParameter() {defaultValue=-0.0235, parameterName=Angle_Right_Offset}
  n11 = Rotator(
   Coordinate=Add(A=OneMinus(None=@n8), B=LinearInterpolate(A=@n9, B=@n10, Alpha=@n3))
   Time=LinearInterpolate(A=Constant(), B=ScalarParameter() {defaultValue=-0.0025, parameterName=Angle_Right}, Alpha=@n3)
  ) {speed=6.2832}
  n12 = ScalarParameter() {defaultValue=128, parameterName=U_Softness_Origin}
  n13 = ScalarParameter() {defaultValue=80, parameterName=U_Softness_Selected}
  n14 = Clamp(None=Multiply(A=ComponentMask(None=@n11) {r=true}, B=LinearInterpolate(A=@n12, B=@n13, Alpha=@n3)))
  n15 = ScalarParameter() {defaultValue=-0.085, parameterName=Angle_Top_Origin}
  n16 = ScalarParameter() {defaultValue=-0.0875, parameterName=Angle_Top_Offset}
  n17 = Rotator(
   Coordinate=Add(A=TextureCoordinate(), B=LinearInterpolate(A=@n15, B=@n16, Alpha=@n3))
   Time=LinearInterpolate(A=Constant() {r=-0.25}, B=ScalarParameter() {defaultValue=-0.245, parameterName=Angle_Top}, Alpha=@n3)
  ) {speed=6.2832}
  n18 = ScalarParameter() {defaultValue=64, parameterName=V_Softness_Origin}
  n19 = ScalarParameter() {defaultValue=40, parameterName=V_Softness_Selected}
  n20 = Clamp(None=Multiply(A=ComponentMask(None=@n17) {r=true}, B=LinearInterpolate(A=@n18, B=@n19, Alpha=@n3)))
  n21 = TextureCoordinate()
  n22 = ScalarParameter() {defaultValue=-0.085, parameterName=Angle_Bottom_Origin}
  n23 = ScalarParameter() {defaultValue=-0.0875, parameterName=Angle_Bottom_Offset}
  n24 = Rotator(
   Coordinate=Add(A=OneMinus(None=@n21), B=LinearInterpolate(A=@n22, B=@n23, Alpha=@n3))
   Time=LinearInterpolate(A=Constant() {r=-0.25}, B=ScalarParameter() {defaultValue=-0.255, parameterName=Angle_Bottom}, Alpha=@n3)
  ) {speed=6.2832}
  n25 = ScalarParameter() {defaultValue=64, parameterName=V_Softness_Origin}
  n26 = ScalarParameter() {defaultValue=40, parameterName=V_Softness_Selected}
  n27 = Clamp(None=Multiply(A=ComponentMask(None=@n24) {r=true}, B=LinearInterpolate(A=@n25, B=@n26, Alpha=@n3)))
  n28 = Clamp(None=DotProduct(A=DotProduct(A=@n7, B=@n14), B=DotProduct(A=@n20, B=@n27)))
  LinearInterpolate(
   A=LinearInterpolate(
    A=VectorParameter() {parameterName=Color_1, defaultValue=<rgba 0.0097 0.2664 1 1>}.RGB
    B=VectorParameter() {parameterName=Color_2, defaultValue=<rgba 0.006 0.0262 0.4508 1>}.RGB
    Alpha=@n28
   )
   B=LinearInterpolate(
    A=VectorParameter() {parameterName=Color_1_Selected, defaultValue=<rgba 0 0.8333 1 1>}.RGB
    B=VectorParameter() {parameterName=Color_2_Selected, defaultValue=<rgba 0.0094 0.1245 0.7 1>}.RGB
    Alpha=@n28
   )
   Alpha=ScalarParameter() {parameterName=AnimateSelection}
  )
 ]
}
